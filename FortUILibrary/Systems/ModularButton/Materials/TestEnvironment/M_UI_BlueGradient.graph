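# Translucent UI material in the ModularButton test environment that renders a two-color radial gradient with a tile pattern color layer.
# An exponential radial gradient on scaled UVs, with adjustable radius and density, blends Color_A and Color_B, optionally sRGB-converted.
# Named reroute inputs, weighted by glow, solid and outline tile amounts, form a mask that blends in Color_1 and Color_2.
# A separate scalar Alpha parameter is also exposed.
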Material /FortUILibrary/Systems/ModularButton/Materials/TestEnvironment/M_UI_BlueGradient
graph {
 properties {
  bUsedWithStaticMesh=false
  blendMode=BLEND_Translucent
  materialDomain=MD_UI
  pixelDepthOffsetMode=PDOM_Legacy
  usedShadingModels=Default Lit
 }
 nodes [
  n1 = MaterialFunctionCall() {materialFunction=&MF_UI_UVScaling}
  n2 = MaterialFunctionCall(Float2=MaterialFunctionCall() {materialFunction=&GetUserInterfaceUV}.Pixel Size) {materialFunction=&BreakOutFloat2Components}
  n3 = Divide(A=@n2.R, B=@n2.R)
  n4 = SquareRoot(None=Add(A=Power(Base=@n3)))
  n5 = MaterialFunctionCall(
   UVs=LinearInterpolate(A=@n1.ScaleX, B=@n1.ScaleX, Alpha=ScalarParameter() {defaultValue=1, parameterName=XY_Priority})
   Radius=Multiply(A=Multiply(A=@n4) {constB=0.5}, B=ScalarParameter() {defaultValue=1, parameterName=RadiusScale})
   Density=ScalarParameter() {defaultValue=2.33, parameterName=Density}
  ) {materialFunction=&RadialGradientExponential}
  n6 = Saturate(None=@n5.RadialGradientExponential)
  n7 = Saturate(None=MaterialFunctionCall!(Linear=@n5.RadialGradientExponential) {materialFunction=/Engine/Functions/Engine_MaterialFunctions01/ImageAdjustment/LinearTosRGB}.sRGB)
  n8 = ScalarParameter() {defaultValue=1, parameterName=Use-sRGB}
  n9 = NamedRerouteUsage()
  n10 = ScalarParameter() {defaultValue=1, group=Overall, parameterName=Amount_GlowTiles}
  n11 = NamedRerouteUsage()
  n12 = ScalarParameter() {defaultValue=1, group=Overall, parameterName=Amount_SolidTiles}
  n13 = Add(
   A=Add(A=Multiply(A=@n9, B=@n10), B=Multiply(A=@n11, B=@n12))
   B=Multiply(A=NamedRerouteUsage(), B=ScalarParameter() {defaultValue=1, group=Overall, parameterName=Amount_OutlineTiles})
  )
  n14 = Saturate(None=Multiply(A=Saturate(None=@n13), B=ScalarParameter() {defaultValue=2, group=Overall, parameterName=Pattern_Inensity}))
  LinearInterpolate(
   A=LinearInterpolate(
    A=VectorParameter() {parameterName=Color_A, defaultValue=<rgba 0 1 0 1>}.RGB
    B=VectorParameter() {parameterName=Color_B, defaultValue=<rgba 1 0 0 1>}.RGB
    Alpha=LinearInterpolate(A=@n6, B=@n7, Alpha=@n8)
   )
   B=LinearInterpolate(
    A=VectorParameter() {group=Overall, parameterName=Color_2, defaultValue=<rgba 0 0.1667 1 1>}.RGB
    B=VectorParameter() {group=Overall, parameterName=Color_1, defaultValue=<rgba 0 1 1 1>}.RGB
    Alpha=@n14
   )
   Alpha=@n14
  )
  ScalarParameter() {defaultValue=1, parameterName=Alpha}
 ]
}
